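# Parent looping stereo MetaSound for cosmetic contrails, layering music, a sweetener loop and glider open, close and landed one-shots from empty wave arrays.
# After a play delay, music passes through a speed converter driving a biquad filter, while the sweetener gets a speed-threshold ADSR before its own converter.
# Glide, freefall and land triggers fire random one-shots into the five-input output mixer.
# The Exited Bus player fires on play but is unconnected to output.
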MetaSoundSource /FortMetasound/Cosmetics/Contrail/MSS_Contrail_Loop_Parent
inputs [
 Exited Bus.Sounds {type=WaveAsset:Array}
 Glider Close.Sounds {type=WaveAsset:Array}
 Glider Landed.Sounds {type=WaveAsset:Array}
 Glider Open.Sounds {type=WaveAsset:Array}
 Master Section. Glider Close Level {type=Float}
 Master Section. Glider Landed Level {type=Float}
 Master Section. Glider Looping SFX Level {type=Float}
 Master Section. Glider Open Level {type=Float}
 Master Section. Music Level {type=Float}
 Mixing.Music..Freq At Max Speed {type=Float}
 Mixing.Music.Bandwidth {type=Float}
 Mixing.Music.Filter Type {type=Enum:BiquadFilterType}
 Mixing.Music.Freq At Min Speed {type=Float}
 Mixing.Music.Frequency At Max Speed {type=Float}
 "Mixing.Music.Gain Multiplier (Low Speed)" {type=Float}
 "Mixing.Music.Gain multiplier (Top Speed)" {type=Float}
 Mixing.Music.Max Freq Gain {type=Float}
 Mixing.Music.Min Freq Gain {type=Float}
 Mixing.SFX.Freq At Min Speed {type=Float}
 Mixing.SFX.Freq Max Speed {type=Float}
 "Mixing.SFX.Gain Multiplier (Low Speed)" {type=Float}
 "Mixing.SFX.Gain Multiplier (Top Speed)" {type=Float}
 Mixing.SFX.Max Freq Gain {type=Float}
 Mixing.SFX.Min Freq Gain {type=Float}
 Movement.Speed {type=Float}
 Movement.SpeedAzimuth {type=Float}
 Movement.SpeedElevation {type=Float}
 Music.Loop.Sounds {type=WaveAsset:Array}
 MusicDelayTime {type=Time}
 OnFreefall {type=Trigger}
 OnGlide {type=Trigger}
 OnLand {type=Trigger}
 Sweetener.Loop.Sounds {type=WaveAsset:Array}
 UE.Source.OnPlay {type=Trigger}
]
interfaces [Fort.Movement:1.0, UE.OutputFormat.Stereo:1.0, UE.Source:1.1]
outputs [
 "UE.OutputFormat.Stereo.Audio:0" {type=Audio}
 "UE.OutputFormat.Stereo.Audio:1" {type=Audio}
]
pages [Default {
  input_defaults {
   Master Section. Glider Close Level=1
   Master Section. Glider Landed Level=1
   Master Section. Glider Looping SFX Level=1
   Master Section. Glider Open Level=1
   Master Section. Music Level=1
   Mixing.Music..Freq At Max Speed=100
   Mixing.Music.Bandwidth=1
   Mixing.Music.Filter Type=3
   Mixing.Music.Freq At Min Speed=0
   Mixing.Music.Frequency At Max Speed=0
   "Mixing.Music.Gain Multiplier (Low Speed)"=1
   "Mixing.Music.Gain multiplier (Top Speed)"=1.3
   Mixing.Music.Max Freq Gain=2
   Mixing.Music.Min Freq Gain=1
   Mixing.SFX.Freq At Min Speed=0
   Mixing.SFX.Freq Max Speed=0
   "Mixing.SFX.Gain Multiplier (Low Speed)"=1
   "Mixing.SFX.Gain Multiplier (Top Speed)"=1.4
   Mixing.SFX.Max Freq Gain=2
   Mixing.SFX.Min Freq Gain=1
   Movement.Speed=0.976
   Movement.SpeedAzimuth=0
   Movement.SpeedElevation=0
   MusicDelayTime=1
   OnFreefall=null
   OnGlide=null
   OnLand=null
   UE.Source.OnPlay=null
   Exited Bus.Sounds []
   Glider Close.Sounds []
   Glider Landed.Sounds []
   Glider Open.Sounds []
   Music.Loop.Sounds []
   Sweetener.Loop.Sounds []
  }
  nodes [
   ADSREnvelopeADSREnvelope_0 = ADSR Envelope.ADSR Envelope:Audio(
    Trigger Release=UE.TriggerOnThreshold:Float(In=^Movement.Speed) {also {defaults {Threshold=5000, Type=1}}}.Out
    Trigger Attack=UE.TriggerOnThreshold:Float(In=^Movement.Speed) {also {defaults {Threshold=5000}}}.Out
   ) {also {defaults {Attack Time=1, Release Time=3.6, Sustain Level=1}}}
   TriggerDelay_0 = UE.Trigger Delay(Delay Time=^MusicDelayTime, In=^UE.Source.OnPlay)
   ArrayRandomGet_4 = Array.Random Get:WaveAsset:Array(In Array=^Sweetener.Loop.Sounds, Next=UE.Reroute(Value=@TriggerDelay_0.Out).Value)
   WavePlayer_0 = UE.Wave Player:Stereo(Play=@ArrayRandomGet_4.On Next, Wave Asset=@ArrayRandomGet_4.Value) {also {defaults {Loop="<MetaSoundBuilderResult.SUCCEEDED: 0>"}, version [1,1]}}
   n1 = &/MSP_Contrail_Speed_Movement_Converter(
    Frequency At Min Speed=^Mixing.SFX.Freq At Min Speed
    Frequency At Max Speed=^Mixing.SFX.Freq Max Speed
    Gain Multiplier At Min Speed=^"Mixing.SFX.Gain Multiplier (Low Speed)"
    "Gain  Multiplier aA Max Speed"=^"Mixing.SFX.Gain Multiplier (Top Speed)"
    Max Freq Gain=^Mixing.SFX.Max Freq Gain
    Min Freq Gain=^Mixing.SFX.Min Freq Gain
    Speed=^Movement.Speed
    In Right=UE.Multiply:Audio(AdditionalOperands=@ADSREnvelopeADSREnvelope_0.Out Envelope, PrimaryOperand=@WavePlayer_0.Out Right).Out
    In Left=UE.Multiply:Audio(AdditionalOperands=@ADSREnvelopeADSREnvelope_0.Out Envelope, PrimaryOperand=@WavePlayer_0.Out Left).Out
   )
   ArrayRandomGet_2 = Array.Random Get:WaveAsset:Array(In Array=^Music.Loop.Sounds, Next=UE.Reroute(Value=@TriggerDelay_0.Out).Value) {also {defaults {Weights []}}}
   WavePlayer_3 = UE.Wave Player:Stereo(Play=@ArrayRandomGet_2.On Next, Wave Asset=@ArrayRandomGet_2.Value) {also {version [1,1]}}
   n2 = &MSP_Contrail_Speed_Movement_Converter(
    Frequency At Max Speed=^Mixing.Music..Freq At Max Speed
    Frequency At Min Speed=^Mixing.Music.Freq At Min Speed
    Max Freq Gain=^Mixing.Music.Max Freq Gain
    Min Freq Gain=^Mixing.Music.Min Freq Gain
    Speed=^Movement.Speed
    In Left=@WavePlayer_3.Out Left
    In Right=@WavePlayer_3.Out Right
   )
   n3 = &StereoBiquadFilter(
    Bandwidth=^Mixing.Music.Bandwidth
    Type=^Mixing.Music.Filter Type
    Gain=@n2.Freq Gain
    In L=@n2.Out Left
    In R=@n2.Out Right
    Cutoff Frequency=UE.Reroute(Value=@n2.Frequency).Value
   )
   ArrayRandomGet_1 = Array.Random Get:WaveAsset:Array(In Array=^Glider Close.Sounds, Next=^OnFreefall) {also {defaults {Weights []}}}
   WavePlayer_1 = UE.Wave Player:Stereo(Play=@ArrayRandomGet_1.On Next, Wave Asset=@ArrayRandomGet_1.Value) {also {version [1,1]}}
   ArrayRandomGet_3 = Array.Random Get:WaveAsset:Array(In Array=^Glider Open.Sounds, Next=^OnGlide)
   WavePlayer_4 = UE.Wave Player:Stereo(Play=@ArrayRandomGet_3.On Next, Wave Asset=@ArrayRandomGet_3.Value) {also {version [1,1]}}
   ArrayRandomGet_0 = Array.Random Get:WaveAsset:Array(In Array=^Glider Landed.Sounds, Next=^OnLand) {also {defaults {Weights []}}}
   WavePlayer_5 = UE.Wave Player:Stereo(Play=@ArrayRandomGet_0.On Next, Wave Asset=@ArrayRandomGet_0.Value) {also {version [1,1]}}
   AudioMixerAudioMixerStereo5_0 = "AudioMixer.Audio Mixer (Stereo, 5)"(
    Gain 3=^Master Section. Glider Close Level
    Gain 4=^Master Section. Glider Landed Level
    Gain 1=^Master Section. Glider Looping SFX Level
    Gain 2=^Master Section. Glider Open Level
    Gain 0=^Master Section. Music Level
    In 1 L=@n1.Out Left
    In 1 R=@n1.Out Right
    In 0 L=@n3.Out L
    In 0 R=@n3.Out R
    In 3 L=@WavePlayer_1.Out Left
    In 3 R=@WavePlayer_1.Out Right
    In 2 L=@WavePlayer_4.Out Left
    In 2 R=@WavePlayer_4.Out Right
    In 4 L=@WavePlayer_5.Out Left
    In 4 R=@WavePlayer_5.Out Right
   )
   ArrayRandomGet_5 = Array.Random Get:WaveAsset:Array(In Array=^Exited Bus.Sounds, Next=^UE.Source.OnPlay) {also {defaults {Weights []}}}
   WavePlayer_2 = UE.Wave Player:Stereo(Play=@ArrayRandomGet_5.On Next, Wave Asset=@ArrayRandomGet_5.Value) {also {version [1,1]}}
  ]
  links [
   {dst="@UE.OutputFormat.Stereo.Audio:0", src=AudioMixerAudioMixerStereo5_0.Out L}
   {dst="@UE.OutputFormat.Stereo.Audio:1", src=AudioMixerAudioMixerStereo5_0.Out R}
  ]
 }]
properties {
 assetImportData=&self:AssetImportData
 audioPropertiesSheet=/BRCosmetics/Sounds/Settings/PropertySheets/Contrail/APS_Cosmetic_Contrail
 bLooping=true
 duration=10000
 numChannels=2
 outputFormat=Stereo
 sampleRate=48000
 soundClassObject=&Fort_SFX_InGame
 soundSubmixObject=&Fortnite_SFX
 virtualizationMode=PlayWhenSilent
 assetUserData [&self:AudioAssetUserData_0]
 soundSubmixSends [{}]
}
