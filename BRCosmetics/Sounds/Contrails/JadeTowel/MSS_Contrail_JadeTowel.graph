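# Stereo looping MetaSound for the JadeTowel cosmetic contrail, a preset of MSS_Contrail_Loop_Parent.
# It supplies only a music loop wave, leaving the sweetener loop, bus exit and glider one-shot arrays empty.
# The parent's inputs expose freefall, glide and land triggers, Fort.Movement speed data, and speed-based gain and filter settings for music and SFX.
# It routes to the Fortnite SFX submix and plays when silent under virtualization.
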MetaSoundSource /BRCosmetics/Sounds/Contrails/JadeTowel/MSS_Contrail_JadeTowel
preset_parent=/FortMetasound/Cosmetics/Contrail/MSS_Contrail_Loop_Parent
inputs [
 Exited Bus.Sounds {type=WaveAsset:Array}
 Glider Close.Sounds {type=WaveAsset:Array}
 Glider Landed.Sounds {type=WaveAsset:Array}
 Glider Open.Sounds {type=WaveAsset:Array}
 Master Section. Glider Close Level {type=Float}
 Master Section. Glider Landed Level {type=Float}
 Master Section. Glider Looping SFX Level {type=Float}
 Master Section. Glider Open Level {type=Float}
 Master Section. Music Level {type=Float}
 Mixing.Music..Freq At Max Speed {type=Float}
 Mixing.Music.Bandwidth {type=Float}
 Mixing.Music.Filter Type {type=Enum:BiquadFilterType}
 Mixing.Music.Freq At Min Speed {type=Float}
 Mixing.Music.Frequency At Max Speed {type=Float}
 "Mixing.Music.Gain Multiplier (Low Speed)" {type=Float}
 "Mixing.Music.Gain multiplier (Top Speed)" {type=Float}
 Mixing.Music.Max Freq Gain {type=Float}
 Mixing.Music.Min Freq Gain {type=Float}
 Mixing.SFX.Freq At Min Speed {type=Float}
 Mixing.SFX.Freq Max Speed {type=Float}
 "Mixing.SFX.Gain Multiplier (Low Speed)" {type=Float}
 "Mixing.SFX.Gain Multiplier (Top Speed)" {type=Float}
 Mixing.SFX.Max Freq Gain {type=Float}
 Mixing.SFX.Min Freq Gain {type=Float}
 Movement.Speed {type=Float}
 Movement.SpeedAzimuth {type=Float}
 Movement.SpeedElevation {type=Float}
 Music.Loop.Sounds {type=WaveAsset:Array}
 MusicDelayTime {type=Time}
 OnFreefall {type=Trigger}
 OnGlide {type=Trigger}
 OnLand {type=Trigger}
 Sweetener.Loop.Sounds {type=WaveAsset:Array}
 UE.Source.OnPlay {type=Trigger}
]
interfaces [Fort.Movement:1.0, UE.OutputFormat.Stereo:1.0, UE.Source:1.1]
outputs [
 "UE.OutputFormat.Stereo.Audio:0" {type=Audio}
 "UE.OutputFormat.Stereo.Audio:1" {type=Audio}
]
pages [Default {input_defaults {
   Master Section. Glider Close Level=1
   Master Section. Glider Landed Level=1
   Master Section. Glider Looping SFX Level=1
   Master Section. Glider Open Level=1
   Master Section. Music Level=1
   Mixing.Music..Freq At Max Speed=100
   Mixing.Music.Bandwidth=1
   Mixing.Music.Filter Type=3
   Mixing.Music.Freq At Min Speed=0
   Mixing.Music.Frequency At Max Speed=0
   "Mixing.Music.Gain Multiplier (Low Speed)"=1
   "Mixing.Music.Gain multiplier (Top Speed)"=1.5
   Mixing.Music.Max Freq Gain=2
   Mixing.Music.Min Freq Gain=1
   Mixing.SFX.Freq At Min Speed=0
   Mixing.SFX.Freq Max Speed=0
   "Mixing.SFX.Gain Multiplier (Low Speed)"=1
   "Mixing.SFX.Gain Multiplier (Top Speed)"=1.4
   Mixing.SFX.Max Freq Gain=2
   Mixing.SFX.Min Freq Gain=1
   Movement.Speed=0
   Movement.SpeedAzimuth=0
   Movement.SpeedElevation=0
   MusicDelayTime=1
   OnFreefall=null
   OnGlide=null
   OnLand=null
   UE.Source.OnPlay=null
   Exited Bus.Sounds []
   Glider Close.Sounds []
   Glider Landed.Sounds []
   Glider Open.Sounds []
   Music.Loop.Sounds [/BRCosmetics/Sounds/Contrails/JadeTowel/wav/MX_IP_JadeTowel_Contrail]
   Sweetener.Loop.Sounds []
  }}]
properties {
 assetImportData=&self:AssetImportData
 audioPropertiesSheet=/BRCosmetics/Sounds/Settings/PropertySheets/Contrail/APS_Cosmetic_Contrail
 bLooping=true
 duration=10000
 numChannels=2
 outputFormat=Stereo
 sampleRate=48000
 soundClassObject=&Fort_SFX_InGame
 soundSubmixObject=&Fortnite_SFX
 virtualizationMode=PlayWhenSilent
 assetUserData [&self:AudioAssetUserData_0]
 soundSubmixSends [{}]
}
